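# Editor utility widget that searches existing permutations of a chosen Material for matches to desired property values.
# Picking a material loads its properties from the first material validation group into editable search parameters in a details view.
# Search scores each permutation by case-insensitive matches on properties flagged bFind, converting that to a percentage and weak/partial/strong label.
# Sorted results fill a second details view, excluding the index-zero reference entry.
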EditorUtilityWidgetBlueprint /TechArtTools/PermutationsSearcher/EUW_PermutationsSearcher
graphs [
 CalculateSimilarity {nodes [
   n1 = CalculateSimilarity!() -> (
    then
    MaterialDatabaseAssetPropertyValue:"Material Database Asset Property Value Structure (by ref)"
    MaterialDatabaseAssetPropertyValue2:"Material Database Asset Property Value Structure (by ref)"
   ) {is=title K2Node_FunctionEntry}
   CompareStrings!(
    LeftStr:String=Break Material Database Asset Property Value!(MaterialDatabaseAssetPropertyValue:"Material Database Asset Property Value Structure (by ref)"=@n1.MaterialDatabaseAssetPropertyValue2) -> (Value:String, bIsModifiedFromParent:bool=~) {is=title K2Node_BreakStruct}
    RightStr:String=Break Material Database Asset Property Value(MaterialDatabaseAssetPropertyValue=@n1) -> (bIsModifiedFromParent=~)
   ) -> (ReturnValue:int=0)
  ]}
 EventGraph {nodes [
   "On Property Changed (MaterialPropertyView)"!() -> (OutputDelegate, then, PropertyName:Name=None) {is=title K2Node_ComponentBoundEvent}
   > Clear(TargetArray:Array of Searchable Material Property Structures=Get SearchableProperties!(self:Material Search Config Object Reference=Get DisplayContainerStaticPropertyStaticSwitch() -> (DisplayContainerStaticPropertyStaticSwitch:Material Search Config Object Reference)) -> (SearchableProperties:Array of Searchable Material Property Structures))
   > SetObject(self=Get DetailView_PermutationResult() -> (DetailView_PermutationResult:Details View Object Reference))
   > Clear(TargetArray=Get Searchable Material Property Permutation Array() -> (Searchable Material Property Permutation Array:Array of Searchable Material Property Structures))
   > Is Valid(InputObject=Get Reference() -> (Reference:Material Object Reference))
   > n1 = GetAllGroups!(execute, bInSyncLoad:bool=true) -> (then, OutGroups:Array of Material Validation Group Object References)
   > n2 = GetMaterialHierarchyInfo!(
    execute
    InGroup:Material Validation Group Object Reference="Get (a copy)"(Array:Array of Material Validation Group Object References=@n1) -> (Output:Material Validation Group Object Reference)
    InBaseMaterialPath:"Soft Object Path Structure (by ref)"=GetSoftObjectPath(Object=Get Reference())
   ) -> (then, OutInfo:Material Database Asset Hierarchy Info Structure)
   > n3 = Set Out Info(Out Info:Material Database Asset Hierarchy Info Structure=@n2) -> (Output_Get:Material Database Asset Hierarchy Info Structure)
   > n4 = GetMaterialProperties!(
    execute
    InInfo:"Material Database Asset Hierarchy Info Structure (by ref)"=Get Out Info() -> (Out Info:Material Database Asset Hierarchy Info Structure)
   ) -> (then, OutProperties:Array of Material Database Asset Property Desc Structures)
   > n5 = GetMaterialPropertyValues!(execute, InInfo:"Material Database Asset Hierarchy Info Structure (by ref)"=@n3, InMaterialIndex:int=0) -> (then, OutValues:Array of Material Database Asset Property Value Structures)
   > n6 = Set ReferencePropertyValues(ReferencePropertyValues:Array of Material Database Asset Property Value Structures=@n5) -> (Output_Get:Array of Material Database Asset Property Value Structures)
   > n7 = For Each Loop(Array:Array of Material Database Asset Property Value Structures=@n6) -> (Array Element:Material Database Asset Property Value Structure)
   n8 = Get DisplayContainerStaticPropertyStaticSwitch()
   Set SearchableProperties!(
    execute=@n7.Completed
    SearchableProperties:Array of Searchable Material Property Structures=Get Searchable Material Property Permutation Array()
    self:Material Search Config Object Reference=@n8
   ) -> (then, Output_Get:Array of Searchable Material Property Structures)
   > n9 = Construct Object from Class!(
    execute
    Class=&/PermutationDataObject.PermutationDataObject_C
    self
    PermutationSearch:Material Search Config Object Reference=@n8
   ) -> (then, ReturnValue:Permutation Data Object Object Reference)
   > Set CurrentReferenceDesiredSearchParameters(CurrentReferenceDesiredSearchParameters:Permutation Data Object Object Reference=@n9) -> (Output_Get:Permutation Data Object Object Reference)
   > SetObject(self=Get DetailViewAvailableParameters() -> (DetailViewAvailableParameters:Details View Object Reference), NewObject=@n9)
   n10 = Branch(execute=@n7.LoopBody, Condition="Equal (Integer)"(A=@n7.Array Index))
   n11 = Break Material Database Asset Property Value(MaterialDatabaseAssetPropertyValue=@n7) -> (bIsModifiedFromParent=~)
   n12 = Break Material Database Asset Property Desc!(MaterialDatabaseAssetPropertyDesc:"Material Database Asset Property Desc Structure (by ref)"="Get (a copy)"(Array:Array of Material Database Asset Property Desc Structures=@n4, Dimension 1=@n7.Array Index) -> (Output:Material Database Asset Property Desc Structure)) -> (Id:Name, Name:Name, Category:EMaterialPropertyCategory Enum) {is=title K2Node_BreakStruct}
   Add(
    execute=@n10.else
    TargetArray:Array of Searchable Material Property Structures=Get Searchable Material Property Permutation Array()
    NewItem:"Searchable Material Property Structure (by ref)"=Make Searchable Material Property!(
     ParameterName:Name=@n12.Name
     Category:EMaterialPropertyType Enum=Select!(
      Default:EMaterialPropertyType Enum=Default
      StaticProperty:EMaterialPropertyType Enum=StaticProperty
      UsageFlag:EMaterialPropertyType Enum=UsageFlag
      StaticSwitch:EMaterialPropertyType Enum=StaticSwitch
      ComponentMask:EMaterialPropertyType Enum=ComponentMask
      Count:EMaterialPropertyType Enum=Default
      Index:EMaterialPropertyCategory Enum=@n12.Category
     ) -> (ReturnValue:EMaterialPropertyType Enum)
     DefaultValue:String=@n11
     bFind:bool=False
     DesiredValue:String=@n11
    ) -> (SearchableMaterialProperty:Searchable Material Property Structure) {is=title K2Node_MakeStruct}
   ) {is=title K2Node_CallArrayFunction}
   "On Clicked (SearchButton)"() {is=title K2Node_ComponentBoundEvent}
   > Clear(TargetArray:Array of Material Permutation Search Result Structures=Get MaterialPermutationSearchResultArray() -> (MaterialPermutationSearchResultArray:Array of Material Permutation Search Result Structures))
   > n13 = For Each Loop(Array:Array of Soft Object Path Structures=Break Material Database Asset Hierarchy Info!(MaterialDatabaseAssetHierarchyInfo:"Material Database Asset Hierarchy Info Structure (by ref)"=Get Out Info()) -> (MaterialPaths:Array of Soft Object Path Structures) {is=title K2Node_BreakStruct}) -> (Array Element:Soft Object Path Structure)
   n14 = Get MaterialPermutationSearchResultArray()
   SortResultsByDistance!(execute=@n13.Completed, Results:Array of Material Permutation Search Result Structures=@n14, bAscending:bool=false) -> (then)
   > n15 = Construct Object from Class!(
    execute
    Class=&/PermutationSearchResultDataObject.PermutationSearchResultDataObject_C
    self
    MaterialPermutationSearchResultArray:Array of Material Permutation Search Result Structures=@n14
   ) -> (then, ReturnValue:Permutation Search Result Data Object Object Reference)
   > SetObject(self=Get DetailView_PermutationResult(), NewObject=@n15)
   Set Score(execute=@n13.LoopBody, Score:int=0) -> (Output_Get:int=0)
   > Set MaxScore(MaxScore:int=0) -> (Output_Get:int=0)
   > Branch(Condition="Not Equal (Integer)"(A=@n13.Array Index))
   > n16 = GetMaterialPropertyValues(InInfo=Get Out Info(), InMaterialIndex=@n13.Array Index)
   > Remove Index!(execute, TargetArray:Array of Material Database Asset Property Value Structures=@n16, IndexToRemove:int=0) -> (then)
   > n17 = For Each Loop(Array:Array of Material Database Asset Property Value Structures=@n16) -> (Array Element:Material Database Asset Property Value Structure)
   n18 = Get MaxScore() -> (MaxScore:int=0)
   n19 = float / float(A="To Float (Integer)"(InInt=Get Score() -> (Score:int=0)), B="To Float (Integer)"(InInt=@n18))
   n20 = SelectInt(A=Truncate(A=float * float(A=@n19, B=100) {is=title K2Node_PromotableOperator}), B=0, bPickA="integer > integer"(A=@n18))
   n21 = SelectText(
    A="LOCTEXT(\"Strong Match\")"
    B="LOCTEXT(\"Partial Match\")"
    bPickA="In Range (Integer)"!(Value:int=@n20, Min:int=60, Max:int=100, InclusiveMin:bool=true, InclusiveMax:bool=true) -> (ReturnValue:bool)
   )
   n22 = "In Range (Integer)"(Value=@n20, Min=0, Max=30, InclusiveMin=true, InclusiveMax=true)
   Add(
    execute=@n17.Completed
    TargetArray:Array of Material Permutation Search Result Structures=Get MaterialPermutationSearchResultArray()
    NewItem:"Material Permutation Search Result Structure (by ref)"=Make Material Permutation Search Result!(
     MaterialPath:Soft Object Path Structure=@n13
     SimilarityRating:Text=SelectText(A="LOCTEXT(\"Weak Match\")", B=@n21, bPickA=@n22)
     SimilarityPercent:int=@n20
    ) -> (MaterialPermutationSearchResult:Material Permutation Search Result Structure) {is=title K2Node_MakeStruct}
   )
   n23 = Get CurrentReferenceDesiredSearchParameters() -> (CurrentReferenceDesiredSearchParameters:Permutation Data Object Object Reference)
   n24 = "Get (a copy)"!(
    Array:Array of Searchable Material Property Structures=Get SearchableProperties(self=Get PermutationSearch!(self:Permutation Data Object Object Reference=@n23) -> (PermutationSearch:Material Search Config Object Reference))
    Dimension 1=@n17.Array Index
   ) -> (
    Output_ParameterName:Name
    Output_Category:EMaterialPropertyType Enum
    Output_DefaultValue:String
    Output_bFind:bool=~
    Output_DesiredValue:String
   )
   Set MaxScore(execute=@n17.LoopBody, MaxScore=int + int(A=Get MaxScore() -> (MaxScore=0), B:bool=@n24.Output_bFind)) -> (Output_Get=0)
   > Branch(Condition="Equal, Case Insensitive (String)"(
     A="Get (a copy)"!(Array:Array of Material Database Asset Property Value Structures=@n16, Dimension 1=@n17.Array Index) -> (Output_Value:String, Output_bIsModifiedFromParent:bool=~)
     B=@n24.Output_DesiredValue
    ))
   > Set Score(Score=int + int(A=Get Score() -> (Score=0), B=@n24.Output_bFind)) -> (Output_Get=0)
   Event PreConstruct()
   n25 = Get MaterialPropertyView() -> (MaterialPropertyView:Single Property View Object Reference)
   > SetObject(self=@n25, NewObject=Self-Reference())
   > SetPropertyName!(execute, self:Single Property View Object Reference=@n25, NewPropertyName:Name=Reference) -> (then)
   GetAllGroups(bInSyncLoad=false)
   GetMaterialHierarchyInfo()
   IsMaterialInGroup!(execute, InGroup:Material Validation Group Object Reference, InMaterial:Material Object Reference) -> (then, bOutIsInGroupPath:bool, bOutIsInGroup:bool)
   IsMaterialInstanceInGroup!(execute, InGroup:Material Validation Group Object Reference, InMaterialInstance:Material Instance Constant Object Reference) -> (then, bOutMaterialInGroup:bool, bOutMaterialPermutationInGroup:bool)
   RemoveInvalidMaterialsFromGroup!(execute, InGroup:Material Validation Group Object Reference) -> (then)
   ResetGroup!(execute, InGroup:Material Validation Group Object Reference) -> (then)
   UpdateMaterialPermutationsInGroup!(execute, InGroup:Material Validation Group Object Reference) -> (then)
   AddMissingMaterialsToGroup!(execute, InGroup:Material Validation Group Object Reference) -> (then)
   GetModifiedShaderCount!(
    execute
    InGroup:Material Validation Group Object Reference
    InMaterial:Material Object Reference
    InModifiedObjects:Array of Material Interface Object References
    InReplacementObjects:Array of Material Interface Object References
    bForceLoadObjects:bool=false
   ) -> (then, ReturnValue:int=0)
   GetShaderCount!(execute, InGroup:Material Validation Group Object Reference, InMaterial:Material Object Reference) -> (then, ReturnValue:int=0)
   GetMaterialPropertyValues(InMaterialIndex=0)
   "In Range (Integer)"(Value=0, Min=30, Max=60, InclusiveMin=true, InclusiveMax=true)
   "To String (Integer)"(InInt=@n20)
   GetAllGroups(bInSyncLoad=false)
   "Get (a ref)"(Array:Array of Material Database Asset Property Desc Structures=GetMaterialProperties(), Dimension 1=0) -> (Output:"Material Database Asset Property Desc Structure (by ref)")
   "Equal (GameplayTagContainer)"(A:Wildcard, B:Wildcard) -> (ReturnValue:bool=~) {is=title K2Node_PromotableOperator}
   Self-Reference()
   Get SearchableProperties(self=Get PermutationSearch(self=Get CurrentReferenceDesiredSearchParameters()))
   Get Reference()
  ]}
]
variables [
 CurrentReferenceDesiredSearchParameters {type=PermutationDataObject_C, default=null}
 DisplayContainerStaticPropertyStaticSwitch {
  type=MaterialSearchConfig
  default=&/EUW_PermutationsSearcher.Default__EUW_PermutationsSearcher_C:MaterialSearchConfig_0
 }
 MaterialPermutationSearchResultArray {type="array<MaterialPermutationSearchResult>", default []}
 MaxScore {type=integer}
 Out Info {type=MaterialDatabaseAssetHierarchyInfo, default {componentMaskNames [], materialAssetDatas [], materialPaths [], staticSwitchNames []}}
 Reference {type=Material, default=null, class=~, graph=~}
 ReferencePropertyValues {type="array<MaterialDatabaseAssetPropertyValue>", default []}
 Score {type=integer}
 Searchable Material Property Permutation Array {type="array<SearchableMaterialProperty>", default []}
 Searchable Material Property Permutation Default {type="array<SearchableMaterialProperty>", default []}
 ValueArray {type="array<string>", default []}
]
widgets {
 info {rootWidgetClass=&GridPanel, widgetCount=10}
 widgets [
  GridPanel_0 {widgetClassPath=&GridPanel, widget=GridPanel_0}
  SizeBox {
   widgetClassPath=&SizeBox
   parent=GridPanel_0
   widget=SizeBox
   properties {
    bOverride_HeightOverride=true
    bOverride_MaxDesiredHeight=true
    bOverride_WidthOverride=true
    heightOverride=800
    maxDesiredHeight=800
    widthOverride=1000
   }
   slot {class=GridSlot, properties {column=1, row=1}}
  }
  DetailView_PermutationResult {
   widgetClassPath=&DetailsView
   parent=SizeBox
   widget=DetailView_PermutationResult
   properties {bAllowFavoriteSystem=true, bAllowFiltering=true, columnWidth=1, viewIdentifier=DetailViewPermutationResult}
   slot {class=SizeBoxSlot}
  }
  TextBlock_47 {
   widgetClassPath=&TextBlock
   parent=GridPanel_0
   widget=TextBlock_47
   properties {
    text=Similar Existing Permutations:
    font {
     bForceMonospaced=false
     bMaterialIsStencil=false
     fontMaterial=null
     fontObject=&Roboto
     letterSpacing=0
     monospacedWidth=1
     size=15
     skewAmount=0
     typefaceFontName=Regular
     outlineSettings {
      bApplyOutlineToDropShadows=false
      bMiteredCorners=false
      bSeparateFillAlpha=false
      outlineMaterial=null
      outlineSize=0
      outlineColor=<rgba 0 0 0 1>
     }
    }
   }
   slot {class=GridSlot, properties {column=1}}
  }
  MaterialPropertyView {
   widgetClassPath=&SinglePropertyView
   parent=GridPanel_0
   widget=MaterialPropertyView
   properties {propertyName=Reference}
   slot {class=GridSlot}
  }
  SizeBox_2 {
   widgetClassPath=&SizeBox
   parent=GridPanel_0
   widget=SizeBox_2
   properties {
    bOverride_HeightOverride=true
    bOverride_MaxDesiredHeight=true
    bOverride_WidthOverride=true
    heightOverride=800
    maxDesiredHeight=800
    widthOverride=600
   }
   slot {class=GridSlot, properties {horizontalAlignment=HAlign_Left, row=1}}
  }
  DetailViewAvailableParameters {
   widgetClassPath=&DetailsView
   parent=SizeBox_2
   widget=DetailViewAvailableParameters
   properties {
    bAllowFavoriteSystem=true
    bAllowFiltering=true
    bShowKeyablePropertiesOption=false
    bShowModifiedPropertiesOption=true
    columnWidth=1
    viewIdentifier=DetailViewAvailableParameters
   }
   slot {class=SizeBoxSlot}
  }
  Overlay {widgetClassPath=&VerticalBox, parent=GridPanel_0, widget=Overlay, slot {class=GridSlot, properties {row=2}}}
  SearchButton {
   widgetClassPath=&EditorUtilityButton
   parent=Overlay
   widget=SearchButton
   properties {colorAndOpacity=<rgba 1 0.8545 0.5531 1>}
   slot {class=VerticalBoxSlot, properties {padding=<ltrb 0 20 0 0>}}
  }
  TextBlock_124 {
   widgetClassPath=&TextBlock
   parent=SearchButton
   widget=TextBlock_124
   properties {
    text=Start Search
    font {
     bForceMonospaced=false
     bMaterialIsStencil=false
     fontMaterial=null
     fontObject=&Roboto
     letterSpacing=0
     monospacedWidth=1
     size=18
     skewAmount=0
     typefaceFontName=Bold
     outlineSettings {
      bApplyOutlineToDropShadows=false
      bMiteredCorners=false
      bSeparateFillAlpha=false
      outlineMaterial=null
      outlineSize=0
      outlineColor=<rgba 0 0 0 1>
     }
    }
   }
   slot {class=ButtonSlot, properties {padding=<ltrb 4 4 4 2>}}
  }
 ]
}
